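# Post-process AnimBP for the SniffRust quadruped cosmetic companion that adds corrective bone deformation for its Stage2 and Stage3 outfit variants.
# The input pose runs through a PoseAsset and is cached, passing through uncorrected outside those variants, which are detected at initialization.
# In outfit, head-to-neck rotation drives additive neck pusher and reverse-jaw offsets, ears are constrained to the jaw bone, and the pusher gets a fixed extra offset.
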AnimBlueprint /CosmeticCompanions/Assets/Quadruped/SniffRust/Meshes/SniffRust_Quadruped_Post_AnimBP
target_skeleton=&/SniffRust_Quadruped_Skeleton
graphs [
 AnimGraph {nodes [
   n1 = Get DeformNeckPushAlpha() -> (DeformNeckPushAlpha:double)
   n2 = "Transform (Modify) Bone - Bone: deform_neck_02_pusher"(ComponentPose=Local To Component(LocalPose=Use cached pose 'CachePose_1'()), Alpha=@n1, Translation="0,2,0", Rotation="0,0,0", Scale="1,1,1") {translationMode=BMM_Additive, translationSpace=BCS_BoneSpace, boneToModify {boneName=deform_neck_02_pusher}, translation {y=2}}
   n3 = "Transform (Modify) Bone - Bone: deform_reverse_jaw"(ComponentPose=@n2, Alpha=@n1, Translation="1,-1,0", Rotation="0,20,0", Scale="1,1,1") {
    rotationMode=BMM_Additive
    rotationSpace=BCS_BoneSpace
    translationMode=BMM_Additive
    translationSpace=BCS_BoneSpace
    boneToModify {boneName=deform_reverse_jaw}
    rotation {yaw=20}
    translation=<1 -1>
   }
   n4 = "Constraint - ear_01_l (P)"(ComponentPose=@n3, Alpha=1, ConstraintWeights_0=1) {boneToModify {boneName=ear_01_l}, constraintSetup [{targetBone {boneName=deform_reverse_jaw}}], constraintWeights [1]}
   n5 = "Constraint - ear_01_r (P)"(ComponentPose=@n4, Alpha=1, ConstraintWeights_0=1) {boneToModify {boneName=ear_01_r}, constraintSetup [{targetBone {boneName=deform_reverse_jaw}}], constraintWeights [1]}
   n6 = "Constraint - ear_02_l (P)"(ComponentPose=@n5, Alpha=1, ConstraintWeights_0=1) {boneToModify {boneName=ear_02_l}, constraintSetup [{targetBone {boneName=deform_reverse_jaw}}], constraintWeights [1]}
   n7 = "Constraint - ear_02_r (P)"(ComponentPose=@n6, Alpha=1, ConstraintWeights_0=1) {boneToModify {boneName=ear_02_r}, constraintSetup [{targetBone {boneName=deform_reverse_jaw}}], constraintWeights [1]}
   n8 = "Transform (Modify) Bone - Bone: deform_neck_02_pusher"(ComponentPose=@n7, Alpha=1, Translation="0,1.7,0", Rotation="0,0,0", Scale="1,1,1") {translationMode=BMM_Additive, translationSpace=BCS_BoneSpace, boneToModify {boneName=deform_neck_02_pusher}, translation {y=1.7}}
   Output Pose(Result=Blend Poses by bool(
     BlendPose_0=Component To Local(ComponentPose=@n8)
     BlendPose_1=Use cached pose 'CachePose_1'()
     bActiveValue=Get IsInOutfit() -> (IsInOutfit:bool)
    ))
   Save cached pose 'CachePose_1'(Pose=SniffRust_Quadruped_PoseAsset!(SourcePose:Pose Link Structure=Input Pose()) -> (Pose:Pose Link Structure) {is=title AnimGraphNode_PoseBlendNode, poseAsset=&/SniffRust_Quadruped_PoseAsset}) {cacheName=CachePose_1}
  ]}
 EventGraph {nodes [
   Event BlueprintInitializeAnimation()
   n1 = GetMainAnimInstance!(self:Companion Post Process Anim Instance Object Reference) -> (ReturnValue:Fort Player Companion Anim Instance Object Reference)
   > Set IsInOutfit(IsInOutfit:bool=OR Boolean(
     A=HasVariantTagForChannel!(
      self:Fort Player Companion Anim Instance Object Reference=@n1
      ChannelTag:Gameplay Tag Structure=(TagName="Cosmetics.Variant.Channel.Outfit")
      VariantTag:Gameplay Tag Structure=(TagName="Cosmetics.Variant.Property.Stage2")
     ) -> (ReturnValue:bool)
     B=HasVariantTagForChannel(self=@n1, ChannelTag=(TagName="Cosmetics.Variant.Channel.Outfit"), VariantTag=(TagName="Cosmetics.Variant.Property.Stage3"))
    )) -> (Output_Get:bool)
   Event BlueprintUpdateAnimation()
   > Branch(Condition=Get IsInOutfit())
   > Set DeformNeckPushAlpha(DeformNeckPushAlpha:double=GetAlphaFromReferenceTransformDelta(
     SkeletalMeshComponent=GetOwningComponent()
     BoneName=head
     BaseName=neck_02
     DeltaComponentType=EulerX
     InputRange=(X=0,Y=-30)
     TwistAxis="0,0,1"
    )) -> (Output_Get:double)
  ]}
]
variables [
 DeformHelmetLiftAlpha {type=number}
 DeformNeckPushAlpha {type=number}
 IsInOutfit {type=boolean}
 PoseAsset {type=PoseAsset, default=&SniffRust_Quadruped_PoseAsset}
]
